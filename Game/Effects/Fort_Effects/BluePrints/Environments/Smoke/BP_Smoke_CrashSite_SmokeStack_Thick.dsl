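; Crash site thick smoke stack environment effect with optional fire, controlled by a single Use fire? boolean.
; Construction script uses that flag to set fire Niagara auto-activation, point light and fake glow visibility, and to pick a glow or no-glow material for the bottom smoke cylinder.
; The setupFxActorOnServer override calls the parent, then copies the flag to the spawned child FX actor.
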(event EventBeginPlay
  (|CallParentFunction))

(fn setupFxActorOnServer
  (params (meshComp :type Static%20Mesh%20Component%20Object%20Reference) (FXActor :type Actor%20Object%20Reference))
  (|CallParentFunction meshComp FXActor)
  (bind _asbp_smoke_crash_site_smoke_stack_thick_child_fx_actor (Utilities|Casting|CastToBP_Smoke_CrashSite_SmokeStack_Thick_ChildFxActor FXActor))
  (Variables|/Game/Effects/Fort_Effects/BluePrints/Environments/Smoke/BP_Smoke_CrashSite_SmokeStack_Thick_ChildFxActor.BP_Smoke_CrashSite_SmokeStack_Thick_ChildFxActor_C|SetUse%20fire? (Variables|Default|GetUse%20fire?) _asbp_smoke_crash_site_smoke_stack_thick_child_fx_actor))

(fn ConstructionScript
  (|CallParentFunction)
  (Utilities|FlowControl|Sequence
    (:then_0
      ;[1 1 1] Turn on/off fire based on the Use fire? boolean
      (Components|Activation|SetAutoActivate (Variables|Default|GetNS_CS_Fire) (Variables|Default|GetUse%20fire?))
      (Rendering|SetVisibility (Variables|Default|GetPointLight) (Variables|Default|GetUse%20fire?) false)
      (Rendering|SetVisibility (Variables|Default|GetFakeGlow) (Variables|Default|GetUse%20fire?) false))
    (:then_2
      ;[1 1 1] Turn on/off glow material based on the Use Fire? boolean
      (Rendering|Material|SetMaterial (Variables|Default|GetSmoke_Cylinder_Bottom) 0 (Selection|SelectMaterialInstance "/Game/Effects/Fort_Effects/Materials/Smoke/MI_SmokeStack_Smoke_Thick_Cylinder_Bottom.MI_SmokeStack_Smoke_Thick_Cylinder_Bottom" "/Game/Effects/Fort_Effects/Materials/Smoke/MI_SmokeStack_Smoke_Thick_Cylinder_Bottom_NoGlow.MI_SmokeStack_Smoke_Thick_Cylinder_Bottom_NoGlow" (Variables|Default|GetUse%20fire?))))))
