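; Child effect actor Blueprint for a thick crash-site smoke stack with optional fire, toggled by the replicated "Use fire?" flag.
; BeginPlay sets the smoke cylinder's material to the glow or no-glow instance to match the flag.
; OnRep updates fire visibility and that material, permanently destroying and clearing the fire component when the flag is false.
; Both paths run only where cosmetic events can execute.
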(event EventBeginPlay
  (Networking|CanExecuteCosmeticEvents
    (:True
      (CallFunction|ValidateFireMaterialParam))))

(fn OnRep_Usefire?
  (bind _use_fire (Variables|Default|GetUse%20fire?))
  (Networking|CanExecuteCosmeticEvents
    (:True
      (Utilities|FlowControl|Sequence
        (:then_0
          ;[1 1 1] Turn on/off fire based on the Use fire? boolean
          (Utilities|IsValid (Variables|Default|GetFire)
            (:Is%20Valid
              (Rendering|SetVisibility (Variables|Default|GetFire) _use_fire false)
              (if (not _use_fire)
                (Components|DestroyComponent (Variables|Default|GetFire))
                (Variables|Default|SetFire 0)))))
        (:then_1
          ;[1 1 1] Turn on/off glow material based on the Use Fire? boolean
          (CallFunction|ValidateFireMaterialParam))))))

(fn ConstructionScript
  (|CallParentFunction))

(fn validateFireMaterialParam
  (Rendering|Material|SetMaterial (Variables|Default|GetSmoke_Cylinder) 0 (Selection|SelectMaterialInstance "/Game/Effects/Fort_Effects/Materials/Smoke/MI_SmokeStack_Smoke_Thick_Cylinder.MI_SmokeStack_Smoke_Thick_Cylinder" "/Game/Effects/Fort_Effects/Materials/Smoke/MI_SmokeStack_Smoke_Thick_Cylinder_NoGlow.MI_SmokeStack_Smoke_Thick_Cylinder_NoGlow" (Variables|Default|GetUse%20fire?))))
