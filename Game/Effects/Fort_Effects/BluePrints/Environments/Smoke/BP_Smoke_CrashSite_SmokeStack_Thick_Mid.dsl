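; Blueprint for a thick, mid-sized crash-site smoke stack environment effect that hands its look settings to a separate FX actor.
; An override of setupFxActorOnServer runs the parent logic, then casts the FX actor to BP_Smoke_CrashSite_SmokeStack_Thick_Mid_Child_FXActor.
; It copies this actor's opacity, base and tip max widths, main color and emissive amount onto that actor.
; BeginPlay, overlap and Tick events are empty, and the construction script only calls its parent.
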(event EventBeginPlay)

(event Collision|EventActorBeginOverlap
  (params (OtherActor :type Actor%20Object%20Reference)))

(event EventTick
  (params (DeltaSeconds :type Float%20%28single-precision%29)))

(fn setupFxActorOnServer
  (params (meshComp :type Static%20Mesh%20Component%20Object%20Reference) (FXActor :type Actor%20Object%20Reference))
  (|CallParentFunction meshComp FXActor)
  (bind _asbp_smoke_crash_site_smoke_stack_thick_mid_child_fxactor (Utilities|Casting|CastToBP_Smoke_CrashSite_SmokeStack_Thick_Mid_Child_FXActor FXActor))
  (Variables|/Game/Effects/Fort_Effects/BluePrints/Environments/Smoke/BP_Smoke_CrashSite_SmokeStack_Thick_Mid_Child_FXActor.BP_Smoke_CrashSite_SmokeStack_Thick_Mid_Child_FXActor_C|SetFinalOpacity (Variables|Default|GetFinalOpacity) _asbp_smoke_crash_site_smoke_stack_thick_mid_child_fxactor)
  (Variables|/Game/Effects/Fort_Effects/BluePrints/Environments/Smoke/BP_Smoke_CrashSite_SmokeStack_Thick_Mid_Child_FXActor.BP_Smoke_CrashSite_SmokeStack_Thick_Mid_Child_FXActor_C|SetWidthBase_Max (Variables|Default|GetWidthBase_Max) _asbp_smoke_crash_site_smoke_stack_thick_mid_child_fxactor)
  (Variables|/Game/Effects/Fort_Effects/BluePrints/Environments/Smoke/BP_Smoke_CrashSite_SmokeStack_Thick_Mid_Child_FXActor.BP_Smoke_CrashSite_SmokeStack_Thick_Mid_Child_FXActor_C|SetWidthTip_Max (Variables|Default|GetWidthTip_Max) _asbp_smoke_crash_site_smoke_stack_thick_mid_child_fxactor)
  (Variables|/Game/Effects/Fort_Effects/BluePrints/Environments/Smoke/BP_Smoke_CrashSite_SmokeStack_Thick_Mid_Child_FXActor.BP_Smoke_CrashSite_SmokeStack_Thick_Mid_Child_FXActor_C|SetMainColor (Variables|Default|GetMainColor) _asbp_smoke_crash_site_smoke_stack_thick_mid_child_fxactor)
  (Variables|/Game/Effects/Fort_Effects/BluePrints/Environments/Smoke/BP_Smoke_CrashSite_SmokeStack_Thick_Mid_Child_FXActor.BP_Smoke_CrashSite_SmokeStack_Thick_Mid_Child_FXActor_C|SetEmissiveAmount (Variables|Default|GetEmissiveAmount) _asbp_smoke_crash_site_smoke_stack_thick_mid_child_fxactor))

(fn ConstructionScript
  (|CallParentFunction))
