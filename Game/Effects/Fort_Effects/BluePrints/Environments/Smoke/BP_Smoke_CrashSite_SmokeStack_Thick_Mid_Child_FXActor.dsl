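; Thick crash-site smoke stack effect actor that drives the look of its sprite mesh material through replicated opacity, color, emissive and width variables.
; BeginPlay and each OnRep handler call Validate functions that write those values to a dynamic material instance, multiplying widths by the mesh's X world scale.
; GetMID reuses or creates that instance, rescaling its tiling and WPO wave intensity to mesh scale, with clamps, only on creation.
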(event EventBeginPlay
  (CallFunction|ValidateOpacity)
  (CallFunction|ValidateWidthBaseMax)
  (CallFunction|ValidateWidthTipMax)
  (CallFunction|ValidateMainColor)
  (CallFunction|ValidateEmissiveAmount))

(fn GetMID
  (bind _asmaterial_instance_dynamic (Utilities|Casting|CastToMaterialInstanceDynamic (Rendering|Material|GetMaterial (Variables|Default|GetSM_Smokestack_Sprites_01) 0))
    (:then
      (return _asmaterial_instance_dynamic))
    (:CastFailed
      (bind _returnvalue (Rendering|Material|CreateDynamicMaterialInstance (Variables|Default|GetSM_Smokestack_Sprites_01) 0 0 "Smoke Crash thick child" :!class /Script/Engine.PrimitiveComponent))
      (CallFunction|ValidateMaterialParams _returnvalue)
      (return _returnvalue))))

(fn OnRep_EmissiveAmount
  (CallFunction|ValidateEmissiveAmount))

(fn OnRep_FinalOpacity
  (CallFunction|ValidateOpacity))

(fn OnRep_MainColor
  (CallFunction|ValidateMainColor))

(fn OnRep_WidthBase_Max
  (CallFunction|ValidateWidthBaseMax))

(fn OnRep_WidthTip_Max
  (CallFunction|ValidateWidthTipMax))

(fn ConstructionScript
  (|CallParentFunction))

(fn ValidateEmissiveAmount
  (bind _returnvalue (Utilities|GetMIDValidated)
    (:Yep
      (Rendering|Material|SetScalarParameterValue _returnvalue "EmissiveAmount" (Variables|Default|GetEmissiveAmount) :!class /Script/Engine.MaterialInstanceDynamic))))

(fn ValidateMainColor
  (bind _returnvalue (Utilities|GetMIDValidated)
    (:Yep
      (Rendering|Material|SetVectorParameterValue _returnvalue "EmissiveColor" (Math|Conversions|ToLinearColor(Color) (Variables|Default|GetMainColor)) :!class /Script/Engine.MaterialInstanceDynamic))))

(fn ValidateMaterialParams
  (params (mid :type Material%20Instance%20Dynamic%20Object%20Reference))
  (LocalVariables|Setl_mid mid)
  (bind _returnvalue (Rendering|Material|GetVectorParameterValue (LocalVariables|Getl_mid) "TileMain" :!class /Script/Engine.MaterialInstanceDynamic))
  (Rendering|Material|SetVectorParameterValue (LocalVariables|Getl_mid) "TileMain" (Math|Conversions|ToLinearColor(Vector) (Math|Vector|MakeVector (Math|Float|Clamp(Float) (* (.x (Math|Conversions|ToVector(LinearColor) _returnvalue)) (/ (.x (Transformation|GetWorldScale (Variables|Default|GetSM_Smokestack_Sprites_01))) 15.0)) 0.5 2.0) (Math|Float|Clamp(Float) (* (.y (Math|Conversions|ToVector(LinearColor) _returnvalue)) (/ (.y (Transformation|GetWorldScale (Variables|Default|GetSM_Smokestack_Sprites_01))) 15.0)) 0.5 2.0) (.z (Math|Conversions|ToVector(LinearColor) _returnvalue)))) :!class /Script/Engine.MaterialInstanceDynamic)
  (bind _returnvalue_2 (Rendering|Material|GetVectorParameterValue (LocalVariables|Getl_mid) "TileDetail" :!class /Script/Engine.MaterialInstanceDynamic))
  (Rendering|Material|SetVectorParameterValue (LocalVariables|Getl_mid) "TileDetail" (Math|Conversions|ToLinearColor(Vector) (Math|Vector|MakeVector (Math|Float|Clamp(Float) (* (.x (Math|Conversions|ToVector(LinearColor) _returnvalue_2)) (/ (.x (Transformation|GetWorldScale (Variables|Default|GetSM_Smokestack_Sprites_01))) 15.0)) 0.5 2.0) (Math|Float|Clamp(Float) (* (.y (Math|Conversions|ToVector(LinearColor) _returnvalue_2)) (/ (.y (Transformation|GetWorldScale (Variables|Default|GetSM_Smokestack_Sprites_01))) 15.0)) 0.5 2.0) (.z (Math|Conversions|ToVector(LinearColor) _returnvalue_2)))) :!class /Script/Engine.MaterialInstanceDynamic)
  (bind _returnvalue_3 (Rendering|Material|GetScalarParameterValue (LocalVariables|Getl_mid) "WPO_Wave_Intenstiy" :!class /Script/Engine.MaterialInstanceDynamic))
  (Rendering|Material|SetScalarParameterValue (LocalVariables|Getl_mid) "WPO_Wave_Intenstiy" (Math|Float|Clamp(Float) (* _returnvalue_3 (/ (.x (Transformation|GetWorldScale (Variables|Default|GetSM_Smokestack_Sprites_01))) 15.0)) 5.0 5e+01) :!class /Script/Engine.MaterialInstanceDynamic))

(fn ValidateOpacity
  (bind _returnvalue (Utilities|GetMIDValidated)
    (:Yep
      (Rendering|Material|SetScalarParameterValue _returnvalue "OpacityFinal" (Variables|Default|GetFinalOpacity) :!class /Script/Engine.MaterialInstanceDynamic))))

(fn ValidateWidthBaseMax
  (bind _returnvalue (Utilities|GetMIDValidated)
    (:Yep
      (Rendering|Material|SetScalarParameterValue _returnvalue "WidthBase_Max" (* (.x (Transformation|GetWorldScale (Variables|Default|GetSM_Smokestack_Sprites_01))) (Variables|Default|GetWidthBase_Max)) :!class /Script/Engine.MaterialInstanceDynamic))))

(fn ValidateWidthTipMax
  (bind _returnvalue (Utilities|GetMIDValidated)
    (:Yep
      (Rendering|Material|SetScalarParameterValue _returnvalue "WidthTip_Max" (* (.x (Transformation|GetWorldScale (Variables|Default|GetSM_Smokestack_Sprites_01))) (Variables|Default|GetWidthTip_Max)) :!class /Script/Engine.MaterialInstanceDynamic))))
