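; Blueprint for a thin smoke stack environment effect at crash sites, built on a parent that manages an FX actor.
; Its server-side setup override calls the parent, then casts the FX actor to the thin smoke stack child class.
; It copies its own opacity, base and tip max width, tiling, main color and emissive values onto that actor.
; The construction script only calls the parent.
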(fn setupFxActorOnServer
  (params (meshComp :type Static%20Mesh%20Component%20Object%20Reference) (FXActor :type Actor%20Object%20Reference))
  (|CallParentFunction meshComp FXActor)
  (bind _asbp_smoke_crash_site_smoke_stack_thin_child_fxactor (Utilities|Casting|CastToBP_Smoke_CrashSite_SmokeStack_Thin_ChildFXActor FXActor))
  (Variables|/Game/Effects/Fort_Effects/BluePrints/Environments/Smoke/BP_Smoke_CrashSite_SmokeStack_Thin_ChildFXActor.BP_Smoke_CrashSite_SmokeStack_Thin_ChildFXActor_C|SetOpacityFinal (Variables|Default|GetOpacityFinal) _asbp_smoke_crash_site_smoke_stack_thin_child_fxactor)
  (Variables|/Game/Effects/Fort_Effects/BluePrints/Environments/Smoke/BP_Smoke_CrashSite_SmokeStack_Thin_ChildFXActor.BP_Smoke_CrashSite_SmokeStack_Thin_ChildFXActor_C|SetWidthBase_Max (Variables|Default|GetWidthBase_Max) _asbp_smoke_crash_site_smoke_stack_thin_child_fxactor)
  (Variables|/Game/Effects/Fort_Effects/BluePrints/Environments/Smoke/BP_Smoke_CrashSite_SmokeStack_Thin_ChildFXActor.BP_Smoke_CrashSite_SmokeStack_Thin_ChildFXActor_C|SetWidthTip_Max (Variables|Default|GetWidthTip_Max) _asbp_smoke_crash_site_smoke_stack_thin_child_fxactor)
  (Variables|/Game/Effects/Fort_Effects/BluePrints/Environments/Smoke/BP_Smoke_CrashSite_SmokeStack_Thin_ChildFXActor.BP_Smoke_CrashSite_SmokeStack_Thin_ChildFXActor_C|SetTiling (Variables|Default|GetTiling) _asbp_smoke_crash_site_smoke_stack_thin_child_fxactor)
  (Variables|/Game/Effects/Fort_Effects/BluePrints/Environments/Smoke/BP_Smoke_CrashSite_SmokeStack_Thin_ChildFXActor.BP_Smoke_CrashSite_SmokeStack_Thin_ChildFXActor_C|SetMainColor (Variables|Default|GetMainColor) _asbp_smoke_crash_site_smoke_stack_thin_child_fxactor)
  (Variables|/Game/Effects/Fort_Effects/BluePrints/Environments/Smoke/BP_Smoke_CrashSite_SmokeStack_Thin_ChildFXActor.BP_Smoke_CrashSite_SmokeStack_Thin_ChildFXActor_C|SetEmissiveAmount (Variables|Default|GetEmissiveAmount) _asbp_smoke_crash_site_smoke_stack_thin_child_fxactor))

(fn ConstructionScript
  (|CallParentFunction))
