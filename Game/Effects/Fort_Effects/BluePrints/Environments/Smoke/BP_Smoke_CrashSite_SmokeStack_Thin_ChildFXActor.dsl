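; Thin crash-site smoke stack environment effect actor, a child FX Blueprint that drives the material of a sprite smokestack mesh.
; GetMID reuses or creates a dynamic material instance, through which RepNotify functions push emissive amount, color, opacity and scale-multiplied widths.
; After the parent construction script, ValidateMaterialParams multiplies V tiling by Tiling and rescales main/detail tiling and wave intensity by mesh world scale, with clamping.
; Event graph events and both tiling RepNotifies are empty.
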(event EventBeginPlay)

(event Collision|EventActorBeginOverlap
  (params (OtherActor :type Actor%20Object%20Reference)))

(event EventTick
  (params (DeltaSeconds :type Float%20%28single-precision%29)))

(fn GetMID
  (bind _asmaterial_instance_dynamic (Utilities|Casting|CastToMaterialInstanceDynamic (Rendering|Material|GetMaterial (Variables|Default|GetSM_Smokestack_Sprites_01) 0))
    (:then
      (return _asmaterial_instance_dynamic))
    (:CastFailed
      (bind _returnvalue (Rendering|Material|CreateDynamicMaterialInstance (Variables|Default|GetSM_Smokestack_Sprites_01) 0 0 "None" :!class /Script/Engine.PrimitiveComponent))
      (return _returnvalue))))

(fn OnRep_EmissiveAmount
  (bind _returnvalue (Default|GetMID))
  (Rendering|Material|SetScalarParameterValue _returnvalue "EmissiveAmount" (Variables|Default|GetEmissiveAmount) :!class /Script/Engine.MaterialInstanceDynamic))

(fn OnRep_MainColor
  (bind _returnvalue (Default|GetMID))
  (Rendering|Material|SetVectorParameterValue _returnvalue "EmissiveColor" (Math|Conversions|ToLinearColor(Color) (Variables|Default|GetMainColor)) :!class /Script/Engine.MaterialInstanceDynamic))

(fn OnRep_OpacityFinal
  (bind _returnvalue (Default|GetMID))
  (Rendering|Material|SetScalarParameterValue _returnvalue "OpacityFinal" (Variables|Default|GetOpacityFinal) :!class /Script/Engine.MaterialInstanceDynamic))

(fn OnRep_Tiling)

(fn OnRep_TilingSmoke_V_Mult)

(fn OnRep_WidthBase_Max
  (bind _returnvalue (Default|GetMID))
  (Rendering|Material|SetScalarParameterValue _returnvalue "WidthBase_Max" (* (.x (Transformation|GetWorldScale (Variables|Default|GetSM_Smokestack_Sprites_01))) (Variables|Default|GetWidthBase_Max)) :!class /Script/Engine.MaterialInstanceDynamic))

(fn OnRep_WidthTip_Max
  (bind _returnvalue (Default|GetMID))
  (Rendering|Material|SetScalarParameterValue _returnvalue "WidthTip_Max" (* (.x (Transformation|GetWorldScale (Variables|Default|GetSM_Smokestack_Sprites_01))) (Variables|Default|GetWidthTip_Max)) :!class /Script/Engine.MaterialInstanceDynamic))

(fn ConstructionScript
  (|CallParentFunction)
  (Default|ValidateMaterialParams))

(fn ValidateMaterialParams
  (bind _returnvalue (Default|GetMID))
  (LocalVariables|Setl_mid _returnvalue)
  (bind _returnvalue_2 (Rendering|Material|GetScalarParameterValue (LocalVariables|Getl_mid) "TilingSmoke_V_Mult" :!class /Script/Engine.MaterialInstanceDynamic))
  (Rendering|Material|SetScalarParameterValue (LocalVariables|Getl_mid) "TilingSmoke_V_Mult" (* _returnvalue_2 (Variables|Default|GetTiling)) :!class /Script/Engine.MaterialInstanceDynamic)
  (bind _returnvalue_3 (Rendering|Material|GetVectorParameterValue (LocalVariables|Getl_mid) "TileMain" :!class /Script/Engine.MaterialInstanceDynamic))
  (Rendering|Material|SetVectorParameterValue (LocalVariables|Getl_mid) "TileMain" (Math|Conversions|ToLinearColor(Vector) (Math|Vector|MakeVector (Math|Float|Clamp(Float) (* (.x (Math|Conversions|ToVector(LinearColor) _returnvalue_3)) (/ (.x (Transformation|GetWorldScale (Variables|Default|GetSM_Smokestack_Sprites_01))) 15.0)) 0.5 2.0) (Math|Float|Clamp(Float) (* (.y (Math|Conversions|ToVector(LinearColor) _returnvalue_3)) (/ (.y (Transformation|GetWorldScale (Variables|Default|GetSM_Smokestack_Sprites_01))) 15.0)) 0.5 2.0) (.z (Math|Conversions|ToVector(LinearColor) _returnvalue_3)))) :!class /Script/Engine.MaterialInstanceDynamic)
  (bind _returnvalue_4 (Rendering|Material|GetVectorParameterValue (LocalVariables|Getl_mid) "TileDetail" :!class /Script/Engine.MaterialInstanceDynamic))
  (Rendering|Material|SetVectorParameterValue (LocalVariables|Getl_mid) "TileDetail" (Math|Conversions|ToLinearColor(Vector) (Math|Vector|MakeVector (Math|Float|Clamp(Float) (* (.x (Math|Conversions|ToVector(LinearColor) _returnvalue_4)) (/ (.x (Transformation|GetWorldScale (Variables|Default|GetSM_Smokestack_Sprites_01))) 15.0)) 0.5 2.0) (Math|Float|Clamp(Float) (* (.y (Math|Conversions|ToVector(LinearColor) _returnvalue_4)) (/ (.y (Transformation|GetWorldScale (Variables|Default|GetSM_Smokestack_Sprites_01))) 15.0)) 0.5 2.0) (.z (Math|Conversions|ToVector(LinearColor) _returnvalue_4)))) :!class /Script/Engine.MaterialInstanceDynamic)
  (bind _returnvalue_5 (Rendering|Material|GetScalarParameterValue (LocalVariables|Getl_mid) "WPO_Wave_Intenstiy" :!class /Script/Engine.MaterialInstanceDynamic))
  (Rendering|Material|SetScalarParameterValue (LocalVariables|Getl_mid) "WPO_Wave_Intenstiy" (Math|Float|Clamp(Float) (* _returnvalue_5 (/ (.x (Transformation|GetWorldScale (Variables|Default|GetSM_Smokestack_Sprites_01))) 15.0)) 5.0 5e+01) :!class /Script/Engine.MaterialInstanceDynamic))
